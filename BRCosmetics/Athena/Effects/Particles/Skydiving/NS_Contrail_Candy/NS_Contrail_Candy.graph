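# Niagara skydiving contrail cosmetic that trails candy boxes, loose candy and two hand-attached mesh ribbons.
# CandyBox spawns swirling boxes on a torus, with curve-driven lid attributes that lift, tumble and shrink late in life.
# Its location events fire in that window, apparently feeding four spawner-less candy emitters (bars, rounds, hearts, stars) that tumble against owner velocity.
# Ribbons follow skeletal mesh bones/sockets and adjust for front-end preview.
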NiagaraSystem /BRCosmetics/Athena/Effects/Particles/Skydiving/NS_Contrail_Candy/NS_Contrail_Candy
dependencies {usedDataInterfaces [&NiagaraDataInterfaceCurve], usedDynamicInputs [], usedModules [&SystemState]}
emitters {
 CandyBox {
  emitterSpawnScript [&FN_PreserveSkydiveUserVariables() {module=PreserveUserVariables}]
  emitterUpdateScript [
   &EmitterState(Enable Distance Culling=true, Module.bUseMaxDistance=true, Module.bUseMinDistance=false, Scalability Mode=::NewEnumerator1)
   &SpawnRate(SpawnRate=2, Use Spawn Probability=true)
  ]
  particleSpawnScript [
   /Niagara/Modules/Spawn/Initialization/InitializeParticle(
    Lifetime=&UniformRangedFloat(Minimum=3.4, Maximum=3.7)
    Mesh Scale=&VectorFromFloat(Value=1)
    Module.Write Color=true
    Module.Write Lifetime=true
    Module.Write Mass=true
    Module.Write Material Random=true
    Module.Write Position=true
    Module.Write Scale=true
    Module.Write SpriteRotation=false
    Module.Write SpriteSize=false
    Position=&Add_Vector(A=/Niagara/DynamicInputs/Helpers/SimulationPosition(), B=<0 0 50>)
   )
   /SetVariables(
    Particles.LidRotCurvScale:NiagaraFloat=&Multiply_Float(
     A=&UniformRangedFloat(Minimum=1, Maximum=2.5)
     B=&UniformAOrBFloat(A=-1)
    )
    Particles.LidZ:NiagaraFloat=&UniformRangedFloat(Minimum=0.9, Maximum=1)
    Particles.StableRandom:NiagaraFloat=&UniformRangedFloat(Minimum=0.1, Maximum=1)
    Particles.StableRandomQuat:Quat4f=/Niagara/DynamicInputs/Rotation/RandomQuaternion()
    Particles.VortexVelocityRate:NiagaraFloat=&UniformRangedFloat(Minimum=36, Maximum=36)
   ) {isSetParameters=true}
   &ApplyInitialForces()
   /Niagara/Modules/Spawn/Orientation/V2/InitialMeshOrientation(
    Rotation=&UniformRangedVector(Minimum=<0 0 0>, Maximum=<0 0 1>)
    Use Orientation Vector=true
    Use Rotation Vector=true
   )
   &TorusLocation(
    Align Torus With Axis=false
    Large Radius=80
    MaskSpawnGroup=false
    Torus Distribution Mode:ENiagaraTorusDistributionMode=::NewEnumerator1
    Torus Mode=::NewEnumerator2
    U Position:NiagaraFloat=&Frac_Float(Float=&Multiply_Float(A=$Emitter.Age, B=1.5))
   )
  ]
  particleUpdateScript [
   &ParticleState()
   /Niagara/Modules/Update/Orientation/V2/UpdateMeshOrientation(Rotation Vector:Vector3f=&MakeVector(X=0, Y=0, Z=$Particles.Initial.StableRandom))
   /Niagara/Modules/Update/Velocity/VortexVelocity(
    Module.Use Influence Falloff=false
    Velocity Amount:NiagaraFloat=$Particles.VortexVelocityRate
    Vortex Origin:Vector3f=/Niagara/Functions/Localspace/SimulationPosition()
   )
   /Niagara/Modules/Update/Forces/CurlNoiseForce(
    Noise Frequency:NiagaraFloat=&FloatFromCurve(FloatCurve=DataInterface() {propertyValues {Curve {keys [<at 0 0.025>,<at 1 0.125>]}}})
    Noise Strength=&FloatFromCurve(FloatCurve=DataInterface() {propertyValues {Curve {keys [<at 0 400>,<at 1.0011 0.002>]}}})
    Pan Noise=false
   )
   &ScaleMeshSize(Scale Factor=&VectorFromFloat(Value=&FloatFromCurve(FloatCurve=DataInterface() {propertyValues {Curve {keys [<at -0.0046 0>,<at 0.0502 1.0034>,<at 0.0995 0.8109>,<at 0.7042 0.808>,<at 0.8163 0.9264>,<at 0.8952 0.9905>,<at 1.0011 0.002>]}}}))) {module=MeshSizeScale}
   &SolveForcesAndVelocity(Clamp Velocity=false, Limit Acceleration=false)
   /SetVariables(
    Particles.LidPosition:Vector3f=&MakeVector(
     X=HlslExpression() {hlslExpression=Particles.Position.x}
     Y=HlslExpression() {hlslExpression=Particles.Position.y}
     Z=&Lerp_Float(
      A=HlslExpression() {hlslExpression=Particles.Position.z}
      B=HlslExpression() {hlslExpression="Particles.Position.z + (Particles.LidZ * 170)"}
      Alpha=&FloatFromCurve(FloatCurve=DataInterface() {propertyValues {Curve {keys [
           <at 0 0>
           {arriveTangent=5.7368, interpMode=RCIM_Cubic, leaveTangent=9.9014, tangentMode=RCTM_Break, time=0.6212, value=0.0058}
           {arriveTangent=3.6481, interpMode=RCIM_Cubic, leaveTangent=3.6481, tangentMode=RCTM_Break, time=0.6603, value=0.251}
           {arriveTangent=2.1213, interpMode=RCIM_Cubic, leaveTangent=2.1213, tangentMode=RCTM_User, time=0.8186, value=0.7636}
           {arriveTangent=0.0493, interpMode=RCIM_Cubic, leaveTangent=3.6639, tangentMode=RCTM_User, time=1.066, value=1.1296}
          ]}}})
     )
    )
    Particles.LidRotation:Quat4f=&LerpQuaternion(
     Quaternion A:Quat4f=$Particles.MeshOrientation
     Quaternion B:Quat4f=&MultiplyQuaternion(Quaternion A=$Particles.MeshOrientation, Quaternion B=$Particles.Initial.StableRandomQuat)
     Lerp Factor:NiagaraFloat=&FloatFromCurve(
      FloatCurve=DataInterface() {propertyValues {Curve {keys [<at 0 0>, {arriveTangent=0.0713, interpMode=RCIM_Cubic, leaveTangent=0.0713, tangentMode=RCTM_User, time=0.64, value=0}, <at 1.25 1>]}}}
      Scale Curve=$Particles.Initial.LidRotCurvScale
     )
     Enable Spherical Linear Interpolation:NiagaraBool=true
    )
    Particles.LidScale:Vector3f=&Lerp_Vector(
     A=$Particles.Scale
     B=<0 0 0>
     Alpha=&FloatFromCurve(FloatCurve=DataInterface() {propertyValues {Curve {keys [
          <at 0 0>
          {arriveTangent=32.5342, interpMode=RCIM_Cubic, leaveTangent=32.5344, tangentMode=RCTM_User, time=0.7307, value=-0.0027}
          <at 0.8123 0.9923 RCIM_Cubic>
         ]}}})
    )
   ) {isSetParameters=true}
   &GenerateLocationEvent(
    Event Send Rate=10
    EventCondition:NiagaraBool=/Niagara/DynamicInputs/Bool/SetBoolByFloatComparison(
     A:NiagaraFloat=&FloatFromCurve(FloatCurve=DataInterface() {propertyValues {Curve {keys [<at 0 0>,<at 0.6511 0.0029>,<at 0.661 1>,<at 0.7482 1>,<at 0.76 0>]}}})
     B:NiagaraFloat=0.9
     ComparisonType:ENiagaraBooleanLogicOps=::NewEnumerator0
    )
    Module.Use Event Send Rate=true
   )
  ]
  rendererClasses [&NiagaraMeshRendererProperties]
  renderers [
   {rendererClass=&NiagaraMeshRendererProperties, rendererIndex=0}
   {rendererClass=&NiagaraMeshRendererProperties, rendererIndex=1}
  ]
 }
 Candy_Bars {
  emitterSpawnScript [&FN_PreserveSkydiveUserVariables() {module=PreserveUserVariables}]
  emitterUpdateScript [&EmitterState(Enable Distance Culling=true, Module.bUseMaxDistance=true, Module.bUseMinDistance=false, Scalability Mode=::NewEnumerator1)]
  particleSpawnScript [
   /Game/Effects/Niagara/Modules/Groups/FN_SetSpawnGroup_v1/FN_SetSpawnGroup(Override SpawnGroup=true, SpawnGroup=5)
   /SetVariables(Particles.VortexAxis:Vector3f=&UniformRangedVector(Minimum=<0 0 -0.5>, Maximum=<0 0 0.5>)) {isSetParameters=true}
   &InitializeParticle(
    Lifetime=&UniformRangedFloat(Minimum=0.65, Maximum=2)
    Mesh Scale=&VectorFromFloat(Value=&UniformRangedFloat(Minimum=0.1, Maximum=0.125))
    Module.Write Color=true
    Module.Write Lifetime=true
    Module.Write Mass=true
    Module.Write Material Random=true
    Module.Write Position=false
    Module.Write Scale=true
    Module.Write SpriteRotation=false
    Module.Write SpriteSize=false
    Position=&SimulationPosition()
   )
   &InitialMeshOrientation(
    Rotation=&UniformRangedVector(Minimum=<0 0 0>, Maximum=<1 1 1>)
    Use Orientation Vector=true
    Use Rotation Vector=true
   )
   &AddRotationalVelocity(Rotation Rate=&UniformRangedVector(Minimum=<0 0 0>, Maximum=<0.5 0.5 0>))
   &AddRotationalVelocity(
    Coordinate Space:ENiagaraMeshTransforms=::NewEnumerator1
    Rotation Rate=&UniformRangedVector(Minimum=<0 0 0.2>, Maximum=<0.1 0.1 0.75>)
   ) {module=AddRotationalVelocity001}
   &AddVelocityInCone(
    Cone Angle=35
    Cone Axis=<0 0 1>
    Module.Use Velocity Falloff On Cone Axis=false
    Velocity Strength=&UniformRangedFloat(Minimum=0, Maximum=150)
   )
   &ApplyInitialForces()
  ]
  particleUpdateScript [
   &ParticleState()
   &ScaleMeshSize(Scale Factor=&VectorFromFloat(Value=&FloatFromCurve(FloatCurve=DataInterface() {propertyValues {Curve {keys [
          <at -0.0058 0.703>
          {arriveTangent=-0.5468, interpMode=RCIM_Cubic, leaveTangent=-0.5468, time=0.89, value=0.6839}
          {arriveTangent=-5.7565, interpMode=RCIM_Cubic, tangentMode=RCTM_Break, time=0.9882, value=0.1594}
         ]}}})))
   &InheritVelocity(
    Module.Limit Speed=true
    Velocity Scale=<-0.01 -0.01 -0.015>
    Velocity Source=&TransformVector(Vector=$Engine.Owner.Velocity, Source Space=::World, Destination Space=::Local)
   )
   &SolveForcesAndVelocity(Clamp Velocity=false, Limit Acceleration=false)
   &SolveRotationalForcesAndVelocity()
  ]
  rendererClasses [&NiagaraMeshRendererProperties]
  renderers [{rendererClass=&NiagaraMeshRendererProperties, rendererIndex=0}]
 }
 Candy_Hearts {
  emitterSpawnScript [&FN_PreserveSkydiveUserVariables() {module=PreserveUserVariables}]
  emitterUpdateScript [&EmitterState(Enable Distance Culling=true, Module.bUseMaxDistance=true, Module.bUseMinDistance=false, Scalability Mode=::NewEnumerator1)]
  particleSpawnScript [
   &FN_SetSpawnGroup(Emitter Bone Count=4)
   /SetVariables() {isSetParameters=true}
   &InitializeParticle(
    Lifetime=&UniformRangedFloat(Minimum=0.8, Maximum=2)
    Mesh Scale=&VectorFromFloat(Value=&UniformRangedFloat(Minimum=0.1, Maximum=0.125))
    Module.Write Color=true
    Module.Write Lifetime=true
    Module.Write Mass=true
    Module.Write Material Random=true
    Module.Write Position=false
    Module.Write Scale=true
    Module.Write SpriteRotation=false
    Module.Write SpriteSize=false
    Position=&SimulationPosition()
   )
   &InitialMeshOrientation(
    Rotation=&UniformRangedVector(Minimum=<0 0 0>, Maximum=<1 1 1>)
    Use Orientation Vector=true
    Use Rotation Vector=true
   )
   &AddRotationalVelocity(Rotation Rate=&UniformRangedVector(Minimum=<0 0 0>, Maximum=<0.5 0.5 0>))
   &AddRotationalVelocity(
    Coordinate Space:ENiagaraMeshTransforms=::NewEnumerator1
    Rotation Rate=&UniformRangedVector(Minimum=<0 0 0.2>, Maximum=<0.1 0.1 0.75>)
   ) {module=AddRotationalVelocity001}
   &AddVelocityInCone(
    Cone Angle=35
    Cone Axis=<0 0 1>
    Module.Use Velocity Falloff On Cone Axis=false
    Velocity Strength=&UniformRangedFloat(Minimum=-15, Maximum=150)
   )
   &ApplyInitialForces()
  ]
  particleUpdateScript [
   &ParticleState()
   &ScaleMeshSize(Scale Factor=&VectorFromFloat(Value=&FloatFromCurve(FloatCurve=DataInterface() {propertyValues {Curve {keys [
          <at -0.0058 0.703>
          {arriveTangent=-0.5468, interpMode=RCIM_Cubic, leaveTangent=-0.5468, time=0.89, value=0.6839}
          {arriveTangent=-5.7565, interpMode=RCIM_Cubic, tangentMode=RCTM_Break, time=0.9882, value=0.1594}
         ]}}})))
   &InheritVelocity(
    Module.Limit Speed=true
    Velocity Scale=<-0.01 -0.01 -0.015>
    Velocity Source=&TransformVector(Vector=$Engine.Owner.Velocity, Source Space=::World, Destination Space=::Local)
   )
   &SolveForcesAndVelocity(Clamp Velocity=false, Limit Acceleration=false)
   &SolveRotationalForcesAndVelocity()
  ]
  rendererClasses [&NiagaraMeshRendererProperties]
  renderers [{rendererClass=&NiagaraMeshRendererProperties, rendererIndex=0}]
 }
 Candy_Round {
  emitterSpawnScript [&FN_PreserveSkydiveUserVariables() {module=PreserveUserVariables}]
  emitterUpdateScript [&EmitterState(Enable Distance Culling=true, Module.bUseMaxDistance=true, Module.bUseMinDistance=false, Scalability Mode=::NewEnumerator1)]
  particleSpawnScript [
   &FN_SetSpawnGroup(Override SpawnGroup=true, SpawnGroup=5)
   /SetVariables() {isSetParameters=true}
   &InitializeParticle(
    Lifetime=&UniformRangedFloat(Minimum=0.65, Maximum=2)
    Mesh Scale=&VectorFromFloat(Value=&UniformRangedFloat(Minimum=0.1, Maximum=0.125))
    Module.Write Color=true
    Module.Write Lifetime=true
    Module.Write Mass=true
    Module.Write Material Random=true
    Module.Write Position=false
    Module.Write Scale=true
    Module.Write SpriteRotation=false
    Module.Write SpriteSize=false
    Position=&SimulationPosition()
   )
   &InitialMeshOrientation(
    Rotation=&UniformRangedVector(Minimum=<0 0 0>, Maximum=<1 1 1>)
    Use Orientation Vector=true
    Use Rotation Vector=true
   )
   &AddRotationalVelocity(Rotation Rate=&UniformRangedVector(Minimum=<0 0 0>, Maximum=<0.5 0.5 0>))
   &AddRotationalVelocity(
    Coordinate Space:ENiagaraMeshTransforms=::NewEnumerator1
    Rotation Rate=&UniformRangedVector(Minimum=<0 0 0.2>, Maximum=<0.1 0.1 0.75>)
   ) {module=AddRotationalVelocity001}
   &AddVelocityInCone(
    Cone Angle=35
    Cone Axis=<0 0 1>
    Module.Use Velocity Falloff On Cone Axis=false
    Velocity Strength=&UniformRangedFloat(Minimum=0, Maximum=150)
   )
   &ApplyInitialForces()
  ]
  particleUpdateScript [
   &ParticleState()
   &ScaleMeshSize(Scale Factor=&VectorFromFloat(Value=&FloatFromCurve(FloatCurve=DataInterface() {propertyValues {Curve {keys [
          <at -0.0058 0.703>
          {arriveTangent=-0.5468, interpMode=RCIM_Cubic, leaveTangent=-0.5468, time=0.89, value=0.6839}
          {arriveTangent=-5.7565, interpMode=RCIM_Cubic, tangentMode=RCTM_Break, time=0.9882, value=0.1594}
         ]}}})))
   &InheritVelocity(
    Module.Limit Speed=true
    Velocity Scale=<-0.01 -0.01 -0.015>
    Velocity Source=&TransformVector(Vector=$Engine.Owner.Velocity, Source Space=::World, Destination Space=::Local)
   )
   &SolveForcesAndVelocity(Clamp Velocity=false, Limit Acceleration=false)
   &SolveRotationalForcesAndVelocity()
  ]
  rendererClasses [&NiagaraMeshRendererProperties]
  renderers [{rendererClass=&NiagaraMeshRendererProperties, rendererIndex=0}]
 }
 Candy_Stars {
  emitterSpawnScript [&FN_PreserveSkydiveUserVariables() {module=PreserveUserVariables}]
  emitterUpdateScript [&EmitterState(Enable Distance Culling=true, Module.bUseMaxDistance=true, Module.bUseMinDistance=false, Scalability Mode=::NewEnumerator1)]
  particleSpawnScript [
   &FN_SetSpawnGroup(Override SpawnGroup=true, SpawnGroup=5)
   /SetVariables() {isSetParameters=true}
   &InitializeParticle(
    Lifetime=&UniformRangedFloat(Minimum=0.65, Maximum=2)
    Mesh Scale=&VectorFromFloat(Value=&UniformRangedFloat(Minimum=0.1, Maximum=0.125))
    Module.Write Color=true
    Module.Write Lifetime=true
    Module.Write Mass=true
    Module.Write Material Random=true
    Module.Write Position=false
    Module.Write Scale=true
    Module.Write SpriteRotation=false
    Module.Write SpriteSize=false
    Position=&SimulationPosition()
   )
   &InitialMeshOrientation(
    Rotation=&UniformRangedVector(Minimum=<0 0 0>, Maximum=<1 1 1>)
    Use Orientation Vector=true
    Use Rotation Vector=true
   )
   &AddRotationalVelocity(Rotation Rate=&UniformRangedVector(Minimum=<0 0 0>, Maximum=<0.5 0.5 0>))
   &AddRotationalVelocity(
    Coordinate Space:ENiagaraMeshTransforms=::NewEnumerator1
    Rotation Rate=&UniformRangedVector(Minimum=<0 0 0.2>, Maximum=<0.1 0.1 0.75>)
   ) {module=AddRotationalVelocity001}
   &AddVelocityInCone(
    Cone Angle=35
    Cone Axis=<0 0 1>
    Module.Use Velocity Falloff On Cone Axis=false
    Velocity Strength=&UniformRangedFloat(Minimum=-15, Maximum=150)
   )
   &ApplyInitialForces()
  ]
  particleUpdateScript [
   &ParticleState()
   &ScaleMeshSize(Scale Factor=&VectorFromFloat(Value=&FloatFromCurve(FloatCurve=DataInterface() {propertyValues {Curve {keys [
          <at -0.0058 0.703>
          {arriveTangent=-0.5468, interpMode=RCIM_Cubic, leaveTangent=-0.5468, time=0.89, value=0.6839}
          {arriveTangent=-5.7565, interpMode=RCIM_Cubic, tangentMode=RCTM_Break, time=0.9882, value=0.1594}
         ]}}})))
   &InheritVelocity(
    Module.Limit Speed=true
    Velocity Scale=<-0.01 -0.01 -0.015>
    Velocity Source=&TransformVector(Vector=$Engine.Owner.Velocity, Source Space=::World, Destination Space=::Local)
   )
   &SolveForcesAndVelocity(Clamp Velocity=false, Limit Acceleration=false)
   &SolveRotationalForcesAndVelocity()
  ]
  rendererClasses [&NiagaraMeshRendererProperties]
  renderers [{rendererClass=&NiagaraMeshRendererProperties, rendererIndex=0}]
 }
 LeftHandRibbon {
  emitterUpdateScript [
   &EmitterState(Inactive Response=::NewEnumerator1, Life Cycle Mode=::NewEnumerator1, Loop Behavior=::NewEnumerator1, Loop Duration=3600, UseLoopDelay=false)
   &SpawnBurst_Instantaneous(Use Spawn Probability=false, UseLoopCountLimit=false)
  ]
  particleSpawnScript [
   /Niagara/Modules/Spawn/Initialization/V2/InitializeParticle(Color Mode=::NewEnumerator0, Mesh Scale Mode=::NewEnumerator3, UsePositionOffset=false, Write Material Random=true) {module=InitializeParticle001}
   /SetVariables(
    Particles.InitialMeshRotationRate:NiagaraFloat=&UniformRangedFloat(Minimum=36, Maximum=300)
    Particles.VortexVelocityRate:NiagaraFloat=&UniformRangedFloat(Minimum=50, Maximum=150)
   ) {isSetParameters=true, module=SetVariables_#1}
   &AddVelocity(Velocity=&FN_FrontEndVector3Modifier(Gameplay=<0 0 0>, Front End=<0 0 500>))
   &DynamicMaterialParameters(
    Index 0 Param 3=&MakeCustomFloatFromBool(Bool=$User.bIsFrontEndPreview, True Float=1, False Float=0)
    Index 0 Param 4=0.65
    Module.Param0WriteEnabled=true
    Module.Param1WriteEnabled=true
    Module.Param2WriteEnabled=true
    Module.Param3WriteEnabled=true
   )
   /Niagara/Modules/Spawn/Orientation/V3/InitialMeshOrientation(Mesh Orientation Mode=::NewEnumerator7, Rotation=<0.8 1.6062 0>, Use Rotation Vector=true)
  ]
  particleUpdateScript [
   &ParticleState(
    Kill Particles When Lifetime Has Elapsed=false
    Let Infinitely Lived Particles Die When Emitter Deactivates=true
    Loop Particles Lifetime=true
   )
   &SolveForcesAndVelocity(Clamp Velocity=false, Limit Acceleration=false)
   &SkeletalMeshLocation(Bone Sampling Mode=::NewEnumerator4, MaskSpawnGroup=false, Orientation Sampling=::NewEnumerator1, Skeletal Mesh=$User.SkeletalMesh)
  ]
  rendererClasses [&NiagaraMeshRendererProperties]
  renderers [{rendererClass=&NiagaraMeshRendererProperties, rendererIndex=0}]
 }
 RightHandRibbon {
  emitterUpdateScript [
   &EmitterState(Inactive Response=::NewEnumerator1, Life Cycle Mode=::NewEnumerator1, Loop Behavior=::NewEnumerator1, Loop Duration=3600, UseLoopDelay=false)
   &SpawnBurst_Instantaneous(Use Spawn Probability=false, UseLoopCountLimit=false)
  ]
  particleSpawnScript [
   &InitializeParticle(Color Mode=::NewEnumerator0, Mesh Scale Mode=::NewEnumerator3, UsePositionOffset=false, Write Material Random=true) {module=InitializeParticle001}
   /SetVariables(
    Particles.InitialMeshRotationRate:NiagaraFloat=&UniformRangedFloat(Minimum=36, Maximum=300)
    Particles.VortexVelocityRate:NiagaraFloat=&UniformRangedFloat(Minimum=50, Maximum=150)
   ) {isSetParameters=true, module=SetVariables_#1}
   &AddVelocity(Velocity=&FN_FrontEndVector3Modifier(Gameplay=<0 0 0>, Front End=<0 0 500>))
   &DynamicMaterialParameters(
    Index 0 Param 3=&MakeCustomFloatFromBool(Bool=$User.bIsFrontEndPreview, True Float=1, False Float=0)
    Index 0 Param 4=0.65
    Module.Param0WriteEnabled=true
    Module.Param1WriteEnabled=true
    Module.Param2WriteEnabled=true
    Module.Param3WriteEnabled=true
   )
   &InitialMeshOrientation(Mesh Orientation Mode=::NewEnumerator7, Rotation=<0.8 -1.6062 0>, Use Rotation Vector=true)
  ]
  particleUpdateScript [
   &ParticleState(
    Kill Particles When Lifetime Has Elapsed=false
    Let Infinitely Lived Particles Die When Emitter Deactivates=true
    Loop Particles Lifetime=true
   )
   &SolveForcesAndVelocity(Clamp Velocity=false, Limit Acceleration=false)
   &SkeletalMeshLocation(
    Bone / Socket Index=1
    Bone Sampling Mode=::NewEnumerator4
    MaskSpawnGroup=false
    Orientation Sampling=::NewEnumerator1
    Skeletal Mesh=$User.SkeletalMesh
   )
  ]
  rendererClasses [&NiagaraMeshRendererProperties]
  renderers [{rendererClass=&NiagaraMeshRendererProperties, rendererIndex=0}]
 }
}
summary {
 systemName=NS_Contrail_Candy
 emitters [
  {emitterName=Candy_Bars, rendererClasses [&NiagaraMeshRendererProperties]}
  {emitterName=Candy_Round, rendererClasses [&NiagaraMeshRendererProperties]}
  {emitterName=Candy_Hearts, rendererClasses [&NiagaraMeshRendererProperties]}
  {emitterName=Candy_Stars, rendererClasses [&NiagaraMeshRendererProperties]}
  {emitterName=CandyBox, rendererClasses [&NiagaraMeshRendererProperties]}
  {emitterName=LeftHandRibbon, rendererClasses [&NiagaraMeshRendererProperties]}
  {emitterName=RightHandRibbon, rendererClasses [&NiagaraMeshRendererProperties]}
 ]
 userVariables [
  User.bIsFrontEnd {type=NiagaraBool}
  User.bIsFrontEndPreview {type=NiagaraBool}
  User.PlayerSpeed {type=NiagaraFloat, defaultValue=0, description=""}
  User.SkeletalMesh {
   type=NiagaraDataInterfaceSkeletalMesh
   defaultValue {
    struct=NiagaraExt_VariableValue_DataInterface
    value {dataInterface=&self:NiagaraDataInterfaceSkeletalMesh_0}
   }
  }
 ]
}
